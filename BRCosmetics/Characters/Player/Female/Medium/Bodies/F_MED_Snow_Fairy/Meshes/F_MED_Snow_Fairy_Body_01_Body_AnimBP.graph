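# Post-process style AnimBP for the Snow Fairy female body cosmetic, copying the parent mesh pose and adding skirt secondary motion.
# Front side chains follow thighs and run AnimDynamics, driving center bones through constraints, with crouch adjustments.
# Back bones blend pelvis and thigh-twist constraints, then run spring AnimDynamics without external force.
# EventGraph derives the front chains' skydiving force from head-to-pelvis direction and float/dive state; the RigidBody node is outside the pose chain.
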AnimBlueprint /BRCosmetics/Characters/Player/Female/Medium/Bodies/F_MED_Snow_Fairy/Meshes/F_MED_Snow_Fairy_Body_01_Body_AnimBP
target_skeleton=&/F_MED_Snow_Fairy_Body_01_Body_Skeleton
graphs [
 AnimGraph {nodes [
   n1 = SelectVector(A=Get skydiving vector force() -> (skydiving vector force:Vector), B="0, 0, 0", bPickA=Get bIsSkyDiving())
   RigidBody(ExternalForce=@n1) {
    bOverrideWorldGravity=true
    overridePhysicsAsset=/BRCosmetics/Characters/Player/Female/Medium/Bodies/F_MED_Octoberfest/Meshes/F_MED_Octoberfest_Dyn_torso
    simulationSpace=BaseBoneSpace
    alphaBoolBlend {blendOption=Linear}
    baseBoneRef {boneName=root}
    componentAppliedLinearAccClamp=<1000 1000 1000>
    componentLinearAccScale {z=0.5}
    componentLinearVelScale {z=0.5}
    overrideWorldGravity {z=-980}
   }
   n2 = "Constraint - dyn_skirt_fr_l_01 (P)"!(ComponentPose=Local To Component(LocalPose=Copy Pose From Mesh() {bUseAttachedParent=true}), Alpha=1) -> (Pose) {
    alphaBoolBlend {blendOption=Linear}
    boneToModify {boneName=dyn_skirt_fr_l_01}
    constraintSetup [{targetBone {boneName=thigh_l}}]
    constraintWeights [1]
   }
   n3 = "Constraint - dyn_skirt_fr_r_01 (P)"!(ComponentPose=@n2, Alpha=1) -> (Pose) {
    alphaBoolBlend {blendOption=Linear}
    boneToModify {boneName=dyn_skirt_fr_r_01}
    constraintSetup [{targetBone {boneName=thigh_r}}]
    constraintWeights [1]
   }
   n4 = "Constraint - dyn_skirt_out_l_01 (P)"!(ComponentPose=@n3, Alpha=1) -> (Pose) {
    alphaBoolBlend {blendOption=Linear}
    boneToModify {boneName=dyn_skirt_out_l_01}
    constraintSetup [{targetBone {boneName=thigh_l}}]
    constraintWeights [1]
   }
   n5 = "Constraint - dyn_skirt_out_r_01 (P)"!(ComponentPose=@n4, Alpha=1) -> (Pose) {
    alphaBoolBlend {blendOption=Linear}
    boneToModify {boneName=dyn_skirt_out_r_01}
    constraintSetup [{targetBone {boneName=thigh_r}}]
    constraintWeights [1]
   }
   n6 = Anim Dynamics - Bone: dyn_skirt_fr_l_01(ComponentPose=@n5, Alpha=1, ExternalForce=@n1) {
    angularDampingOverride=0.98
    bGravityOverrideInSimSpace=true
    bOverrideAngularDamping=true
    bUseGravityOverride=true
    numSolverIterationsPostUpdate=2
    numSolverIterationsPreUpdate=8
    simulationSpace=BoneRelative
    alphaBoolBlend {blendOption=Linear}
    boundBone {boneName=dyn_skirt_fr_l_01}
    gravityOverride {z=-980}
    physicsBodyDefinitions [{
      sphereCollisionRadius=0
      boundBone {boneName=dyn_skirt_fr_l_01}
      boxExtents=<20 20 10>
      constraintSetup {twistAxis=AxisZ, angularLimitsMax=<0 0 2>, angularLimitsMin=<0 -30 -2>, angularTarget=<0 0 0>}
      localJointOffset=<0 0 -3>
     }]
    relativeSpaceBone {boneName=root}
   }
   n7 = Anim Dynamics - Bone: dyn_skirt_fr_l_02(ComponentPose=@n6, Alpha=1, ExternalForce=@n1) {
    angularDampingOverride=0.98
    bGravityOverrideInSimSpace=true
    bOverrideAngularDamping=true
    bUseGravityOverride=true
    simulationSpace=BoneRelative
    alphaBoolBlend {blendOption=Linear}
    boundBone {boneName=dyn_skirt_fr_l_02}
    gravityOverride {z=-980}
    physicsBodyDefinitions [{
      sphereCollisionRadius=0
      boundBone {boneName=dyn_skirt_fr_l_02}
      boxExtents=<20 20 10>
      constraintSetup {twistAxis=AxisZ, angularLimitsMax=<10 1 2>, angularLimitsMin=<-10 -30 -2>, angularTarget=<0 0 0>}
      localJointOffset=<0 0 -3>
     }]
    relativeSpaceBone {boneName=root}
   }
   n8 = Anim Dynamics - Bone: dyn_skirt_fr_l_03(ComponentPose=@n7, Alpha=1, ExternalForce=@n1) {
    angularDampingOverride=0.98
    bGravityOverrideInSimSpace=true
    bOverrideAngularDamping=true
    bUseGravityOverride=true
    simulationSpace=BoneRelative
    alphaBoolBlend {blendOption=Linear}
    boundBone {boneName=dyn_skirt_fr_l_03}
    gravityOverride {z=-980}
    physicsBodyDefinitions [{
      sphereCollisionRadius=0
      boundBone {boneName=dyn_skirt_fr_l_03}
      boxExtents=<20 20 10>
      constraintSetup {twistAxis=AxisZ, angularLimitsMax=<10 0 2>, angularLimitsMin=<-10 -30 -2>, angularTarget=<0 0 0>}
      localJointOffset=<0 0 -3>
     }]
    relativeSpaceBone {boneName=root}
   }
   n9 = Anim Dynamics - Bone: dyn_skirt_fr_l_04(ComponentPose=@n8, Alpha=1, ExternalForce=@n1) {
    angularDampingOverride=0.98
    bGravityOverrideInSimSpace=true
    bOverrideAngularDamping=true
    bUseGravityOverride=true
    simulationSpace=BoneRelative
    alphaBoolBlend {blendOption=Linear}
    boundBone {boneName=dyn_skirt_fr_l_04}
    gravityOverride {z=-980}
    physicsBodyDefinitions [{
      sphereCollisionRadius=0
      boundBone {boneName=dyn_skirt_fr_l_04}
      boxExtents=<20 20 10>
      constraintSetup {twistAxis=AxisZ, angularLimitsMax=<0 0 0>, angularLimitsMin=<0 -30 0>, angularTarget=<0 0 0>}
      localJointOffset=<0 0 -3>
     }]
    relativeSpaceBone {boneName=root}
   }
   n10 = Anim Dynamics - Bone: dyn_skirt_fr_r_01(ComponentPose=@n9, Alpha=1, ExternalForce=@n1) {
    angularDampingOverride=0.98
    bGravityOverrideInSimSpace=true
    bOverrideAngularDamping=true
    bUseGravityOverride=true
    simulationSpace=BoneRelative
    alphaBoolBlend {blendOption=Linear}
    boundBone {boneName=dyn_skirt_fr_r_01}
    gravityOverride {z=-980}
    physicsBodyDefinitions [{
      sphereCollisionRadius=0
      boundBone {boneName=dyn_skirt_fr_r_01}
      boxExtents=<20 20 10>
      constraintSetup {twistAxis=AxisZ, angularLimitsMax=<0 0 2>, angularLimitsMin=<0 -30 -2>, angularTarget=<0 0 0>}
      localJointOffset=<0 0 3>
     }]
    relativeSpaceBone {boneName=root}
   }
   n11 = Anim Dynamics - Bone: dyn_skirt_fr_r_02(ComponentPose=@n10, Alpha=1, ExternalForce=@n1) {
    angularDampingOverride=0.98
    bGravityOverrideInSimSpace=true
    bOverrideAngularDamping=true
    bUseGravityOverride=true
    simulationSpace=BoneRelative
    alphaBoolBlend {blendOption=Linear}
    boundBone {boneName=dyn_skirt_fr_r_02}
    gravityOverride {z=-980}
    physicsBodyDefinitions [{
      sphereCollisionRadius=0
      boundBone {boneName=dyn_skirt_fr_r_02}
      boxExtents=<20 20 10>
      constraintSetup {twistAxis=AxisZ, angularLimitsMax=<10 1 2>, angularLimitsMin=<-10 -30 -2>, angularTarget=<0 0 0>}
      localJointOffset=<0 0 3>
     }]
    relativeSpaceBone {boneName=root}
   }
   n12 = Anim Dynamics - Bone: dyn_skirt_fr_r_03(ComponentPose=@n11, Alpha=1, ExternalForce=@n1) {
    angularDampingOverride=0.98
    bGravityOverrideInSimSpace=true
    bOverrideAngularDamping=true
    bUseGravityOverride=true
    simulationSpace=BoneRelative
    alphaBoolBlend {blendOption=Linear}
    boundBone {boneName=dyn_skirt_fr_r_03}
    gravityOverride {z=-980}
    physicsBodyDefinitions [{
      sphereCollisionRadius=0
      boundBone {boneName=dyn_skirt_fr_r_03}
      boxExtents=<20 20 10>
      constraintSetup {twistAxis=AxisZ, angularLimitsMax=<10 3 2>, angularLimitsMin=<-10 -30 -2>, angularTarget=<0 0 0>}
      localJointOffset=<0 0 3>
     }]
    relativeSpaceBone {boneName=root}
   }
   n13 = Anim Dynamics - Bone: dyn_skirt_fr_r_04(ComponentPose=@n12, Alpha=1, ExternalForce=@n1) {
    angularDampingOverride=0.98
    bGravityOverrideInSimSpace=true
    bOverrideAngularDamping=true
    bUseGravityOverride=true
    simulationSpace=BoneRelative
    alphaBoolBlend {blendOption=Linear}
    boundBone {boneName=dyn_skirt_fr_r_04}
    gravityOverride {z=-980}
    physicsBodyDefinitions [{
      sphereCollisionRadius=0
      boundBone {boneName=dyn_skirt_fr_r_04}
      boxExtents=<20 20 10>
      constraintSetup {twistAxis=AxisZ, angularLimitsMax=<0 0 0>, angularLimitsMin=<0 -30 0>, angularTarget=<0 0 0>}
      localJointOffset=<0 0 3>
     }]
    relativeSpaceBone {boneName=root}
   }
   n14 = "Constraint - dyn_skirt_fr_01 (PP)"!(ComponentPose=@n13, Alpha=SelectFloat(A=0.75, B=1, bPickA=Get bIsCrouching()), ConstraintWeights_0=1, ConstraintWeights_1:float=1) -> (Pose) {
    alphaBoolBlend {blendOption=Linear}
    boneToModify {boneName=dyn_skirt_fr_01}
    constraintSetup [{targetBone {boneName=dyn_skirt_fr_l_01}}, {targetBone {boneName=dyn_skirt_fr_r_01}}]
    constraintWeights [1,1]
   }
   n15 = "Constraint - dyn_skirt_fr_02 (RR)"!(ComponentPose=@n14, Alpha=1, ConstraintWeights_0=1, ConstraintWeights_1:float=1) -> (Pose) {
    alphaBoolBlend {blendOption=Linear}
    boneToModify {boneName=dyn_skirt_fr_02}
    constraintSetup [{transformType=Rotation, targetBone {boneName=dyn_skirt_fr_l_02}}, {transformType=Rotation, targetBone {boneName=dyn_skirt_fr_r_02}}]
    constraintWeights [1,1]
   }
   n16 = "Constraint - dyn_skirt_fr_03 (RR)"!(ComponentPose=@n15, Alpha=1, ConstraintWeights_0=1, ConstraintWeights_1:float=1) -> (Pose) {
    alphaBoolBlend {blendOption=Linear}
    boneToModify {boneName=dyn_skirt_fr_03}
    constraintSetup [{transformType=Rotation, targetBone {boneName=dyn_skirt_fr_l_03}}, {transformType=Rotation, targetBone {boneName=dyn_skirt_fr_r_03}}]
    constraintWeights [1,1]
   }
   n17 = "Constraint - dyn_skirt_fr_04 (RR)"!(ComponentPose=@n16, Alpha=1, ConstraintWeights_0=1, ConstraintWeights_1:float=1) -> (Pose) {
    alphaBoolBlend {blendOption=Linear}
    boneToModify {boneName=dyn_skirt_fr_04}
    constraintSetup [{transformType=Rotation, targetBone {boneName=dyn_skirt_fr_l_04}}, {transformType=Rotation, targetBone {boneName=dyn_skirt_fr_r_04}}]
    constraintWeights [1,1]
   }
   n18 = "Transform (Modify) Bone - Bone: dyn_skirt_fr_l_01"(
    ComponentPose=@n17
    Alpha="To Float (Boolean)"(InBool=Get bIsCrouching())
    Translation="6,4.159,-3.475"
    Rotation="-17.790,-8.095,2.266"
    Scale="1,1,1"
   ) {
    rotationMode=BMM_Additive
    rotationSpace=BCS_BoneSpace
    translationMode=BMM_Additive
    translationSpace=BCS_BoneSpace
    alphaBoolBlend {blendOption=Linear}
    boneToModify {boneName=dyn_skirt_fr_l_01}
    rotation=<rot -17.79 -8.0953 2.266>
    translation=<6 4.159 -3.475>
   }
   n19 = "Constraint - dyn_skirt_bk_l_01 (PP)"!(ComponentPose=@n18, Alpha=1, ConstraintWeights_0=0.9271, ConstraintWeights_1:float=0.0729) -> (Pose) {
    alphaBoolBlend {blendOption=Linear}
    boneToModify {boneName=dyn_skirt_bk_l_01}
    constraintSetup [{targetBone {boneName=pelvis}}, {targetBone {boneName=thigh_twist_01_l}}]
    constraintWeights [1,1]
   }
   n20 = "Constraint - dyn_skirt_bk_01 (PPP)"!(ComponentPose=@n19, Alpha=1, ConstraintWeights_0=0.9697, ConstraintWeights_1:float=0.0193, ConstraintWeights_2:float=0.011) -> (Pose) {
    alphaBoolBlend {blendOption=Linear}
    boneToModify {boneName=dyn_skirt_bk_01}
    constraintSetup [{targetBone {boneName=pelvis}}, {targetBone {boneName=thigh_twist_01_r}}, {targetBone {boneName=thigh_twist_01_l}}]
    constraintWeights [1,1,1]
   }
   n21 = "Constraint - dyn_skirt_bk_r_01 (PP)"!(ComponentPose=@n20, Alpha=1, ConstraintWeights_0=0.9271, ConstraintWeights_1:float=0.0729) -> (Pose) {
    alphaBoolBlend {blendOption=Linear}
    boneToModify {boneName=dyn_skirt_bk_r_01}
    constraintSetup [{targetBone {boneName=pelvis}}, {targetBone {boneName=thigh_twist_01_r}}]
    constraintWeights [1,1]
   }
   n22 = Anim Dynamics - Bone: dyn_skirt_bk_l_01(ComponentPose=@n21, Alpha=1, ExternalForce="0,0,0") {
    angularDampingOverride=0.98
    angularSpringConstant=50
    bAngularSpring=true
    bGravityOverrideInSimSpace=true
    bOverrideAngularDamping=true
    simulationSpace=RootRelative
    alphaBoolBlend {blendOption=Linear}
    boundBone {boneName=dyn_skirt_bk_l_01}
    componentAppliedLinearAccClamp=<3000 3000 3000>
    componentLinearAccScale=<0.3 0.3 0.3>
    componentLinearVelScale=<0.3 0.3 0.3>
    gravityOverride {z=-980}
    physicsBodyDefinitions [{
      sphereCollisionRadius=0
      boundBone {boneName=dyn_skirt_bk_l_01}
      boxExtents=<20 15 20>
      constraintSetup {twistAxis=AxisZ, angularLimitsMax=<5 0 0>, angularLimitsMin=<-5 -30 0>, angularTarget=<0 0 0>}
      localJointOffset=<0 0 -1>
     }]
    relativeSpaceBone {boneName=root}
   }
   n23 = Anim Dynamics - Bone: dyn_skirt_bk_l_02(ComponentPose=@n22, Alpha=1, ExternalForce="0,0,0") {
    angularDampingOverride=0.98
    angularSpringConstant=50
    bAngularSpring=true
    bGravityOverrideInSimSpace=true
    bOverrideAngularDamping=true
    simulationSpace=RootRelative
    alphaBoolBlend {blendOption=Linear}
    boundBone {boneName=dyn_skirt_bk_l_02}
    componentAppliedLinearAccClamp=<3000 3000 3000>
    componentLinearAccScale=<0.3 0.3 0.3>
    componentLinearVelScale=<0.3 0.3 0.3>
    gravityOverride {z=-980}
    physicsBodyDefinitions [{
      sphereCollisionRadius=0
      boundBone {boneName=dyn_skirt_bk_l_02}
      boxExtents=<20 15 20>
      constraintSetup {twistAxis=AxisZ, angularLimitsMax=<5 0 0>, angularLimitsMin=<-5 -30 0>, angularTarget=<0 0 0>}
      localJointOffset=<0 0 -1>
     }]
    relativeSpaceBone {boneName=root}
   }
   n24 = Anim Dynamics - Bone: dyn_skirt_bk_01(ComponentPose=@n23, Alpha=1, ExternalForce="0,0,0") {
    angularDampingOverride=0.98
    angularSpringConstant=50
    bAngularSpring=true
    bGravityOverrideInSimSpace=true
    bOverrideAngularDamping=true
    simulationSpace=RootRelative
    alphaBoolBlend {blendOption=Linear}
    boundBone {boneName=dyn_skirt_bk_01}
    componentAppliedLinearAccClamp=<3000 3000 3000>
    componentLinearAccScale=<0.3 0.3 0.3>
    componentLinearVelScale=<0.3 0.3 0.3>
    gravityOverride {z=-980}
    physicsBodyDefinitions [{
      sphereCollisionRadius=0
      boundBone {boneName=dyn_skirt_bk_01}
      boxExtents=<20 15 20>
      constraintSetup {twistAxis=AxisZ, angularLimitsMax=<5 0 0>, angularLimitsMin=<-5 -30 0>, angularTarget=<0 0 0>}
      localJointOffset=<0 0 -1>
     }]
    relativeSpaceBone {boneName=root}
   }
   n25 = Anim Dynamics - Bone: dyn_skirt_bk_02(ComponentPose=@n24, Alpha=1, ExternalForce="0,0,0") {
    angularDampingOverride=0.98
    angularSpringConstant=50
    bAngularSpring=true
    bGravityOverrideInSimSpace=true
    bOverrideAngularDamping=true
    simulationSpace=RootRelative
    alphaBoolBlend {blendOption=Linear}
    boundBone {boneName=dyn_skirt_bk_02}
    componentAppliedLinearAccClamp=<3000 3000 3000>
    componentLinearAccScale=<0.3 0.3 0.3>
    componentLinearVelScale=<0.3 0.3 0.3>
    gravityOverride {z=-980}
    physicsBodyDefinitions [{
      sphereCollisionRadius=0
      boundBone {boneName=dyn_skirt_bk_02}
      boxExtents=<20 15 20>
      constraintSetup {twistAxis=AxisZ, angularLimitsMax=<5 0 0>, angularLimitsMin=<-5 -30 0>, angularTarget=<0 0 0>}
      localJointOffset=<0 0 -1>
     }]
    relativeSpaceBone {boneName=root}
   }
   n26 = Anim Dynamics - Bone: dyn_skirt_bk_r_01(ComponentPose=@n25, Alpha=1, ExternalForce="0,0,0") {
    angularDampingOverride=0.98
    angularSpringConstant=50
    bAngularSpring=true
    bGravityOverrideInSimSpace=true
    bOverrideAngularDamping=true
    simulationSpace=RootRelative
    alphaBoolBlend {blendOption=Linear}
    boundBone {boneName=dyn_skirt_bk_r_01}
    componentAppliedLinearAccClamp=<3000 3000 3000>
    componentLinearAccScale=<0.3 0.3 0.3>
    componentLinearVelScale=<0.3 0.3 0.3>
    gravityOverride {z=-980}
    physicsBodyDefinitions [{
      sphereCollisionRadius=0
      boundBone {boneName=dyn_skirt_bk_r_01}
      boxExtents=<20 15 20>
      constraintSetup {twistAxis=AxisZ, angularLimitsMax=<5 0 0>, angularLimitsMin=<-5 -30 0>, angularTarget=<0 0 0>}
      localJointOffset=<0 0 1>
     }]
    relativeSpaceBone {boneName=root}
   }
   n27 = Anim Dynamics - Bone: dyn_skirt_bk_r_02(ComponentPose=@n26, Alpha=1, ExternalForce="0,0,0") {
    angularDampingOverride=0.98
    angularSpringConstant=50
    bAngularSpring=true
    bGravityOverrideInSimSpace=true
    bOverrideAngularDamping=true
    simulationSpace=RootRelative
    alphaBoolBlend {blendOption=Linear}
    boundBone {boneName=dyn_skirt_bk_r_02}
    componentAppliedLinearAccClamp=<3000 3000 3000>
    componentLinearAccScale=<0.3 0.3 0.3>
    componentLinearVelScale=<0.3 0.3 0.3>
    gravityOverride {z=-980}
    physicsBodyDefinitions [{
      sphereCollisionRadius=0
      boundBone {boneName=dyn_skirt_bk_r_02}
      boxExtents=<20 15 20>
      constraintSetup {twistAxis=AxisZ, angularLimitsMax=<5 0 0>, angularLimitsMin=<-5 -30 0>, angularTarget=<0 0 0>}
      localJointOffset=<0 0 1>
     }]
    relativeSpaceBone {boneName=root}
   }
   Output Pose(Result=Component To Local(ComponentPose=@n27)) {layerGroup=None}
   Get isInFrontend() -> (isInFrontend:double)
  ]}
 EventGraph {nodes [
   Event BlueprintUpdateAnimation()
   > n1 = Sequence()
   > K2Node_Composite_0 = SkirtGraph!(NewParam:Exec, then_0:Exec) {is=title K2Node_Composite}
   Set isInFrontend(execute=@n1.then_1, isInFrontend:double="To Float (Boolean)"(InBool=Get bIsInFrontEnd())) -> (Output_Get:double)
  ]}
 SkirtGraph {
  path=EventGraph.K2Node_Composite_0.SkirtGraph
  nodes [
   Inputs!() -> (NewParam:Exec, then_0:Exec)
   > n1 = Branch(Condition=Get bCanPlayCustomAnimations() -> (bCanPlayCustomAnimations:bool))
   n2 = Branch(execute=[@n1.else, @n1.then], Condition=Get bIsSkyDiving())
   n3 = Branch(execute=[@n2.else, @n2.then], Condition=Get bIsSkydiveFloating())
   n4 = Get Direction Between Sockets!(
    Component:Skeletal Mesh Component Object Reference=GetOwnerSkeletalMeshComponent()
    SocketOrBoneNameFrom:Name=head
    SocketOrBoneNameTo:Name=pelvis
   ) -> (ReturnValue:Vector) {is=title K2Node_CallFunction}
   > Set skydiving vector force(skydiving vector force:Vector=vector + vector(
     A:Vector=vector * vector(A:Vector=@n4, B:Vector="75,75,75") -> (ReturnValue:Vector) {is=title K2Node_CommutativeAssociativeBinaryOperator}
     B:Vector="0, 0,2000"
    ) -> (ReturnValue:Vector) {is=title K2Node_CommutativeAssociativeBinaryOperator}) -> (Output_Get:Vector)
   n5 = Branch(execute=@n3.else, Condition=Get bIsSkydiveDiveMode())
   > Set skydiving vector force(skydiving vector force=vector + vector(A=vector * vector(A=@n4, B="180,180,200"), B="0, 0,2000"))
   Set skydiving vector force(execute=@n5.else, skydiving vector force=vector * vector(A=@n4, B="-15,-15,-15"))
   TryGetPawnOwner()
   Outputs()
  ]
 }
]
variables [
 isInFrontend {type=number}
 skydiving vector force {type=Vector}
]
