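# Physics asset for the Snow Fairy face accessory on the medium female body, defining physics for two dangling strings.
# Left and right six-bone capsule chains attach to the head with wide limits, with tight joints along each chain whose second swing limit widens toward the tip.
# Head, neck, clavicle and upper-spine capsules, plus large angled spine boxes, form body collision volumes, linked by free-rotation constraints.
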PhysicsAsset /BRCosmetics/Characters/Player/Female/Medium/Bodies/F_MED_Snow_Fairy/Meshes/F_MED_Snow_Fairy_FaceAcc_01_Physics
bodies {
 clavicle_l [
  {shapeType=Capsule, length=7.505, radius=4.88, center=<7.1569 0 2.8409>, rotation=<rot 0 -89.9999 89.9999>}
  {shapeType=Capsule, length=24.75, radius=9.5, shapeName=clavicle_l_capsule, center=<22.8926 9.624 9.476>, rotation=<rot 0 -15.0012 89.9995>}
 ]
 clavicle_r [
  {shapeType=Capsule, length=7.255, radius=5.13, center=<-5.5993 0.7191 -2.2363>, rotation=<rot 0 -89.9999 89.9999>}
  {
   shapeType=Capsule
   length=24.75
   radius=9.5
   shapeName=clavicle_r_capsule
   center=<-22.8926 -9.624 -9.476>
   rotation=<rot 0.0002 -15 90.0007>
  }
 ]
 dyn_string_l_01 [{shapeType=Capsule, length=0, radius=3, center=<1 0 0>}]
 dyn_string_l_02 [{shapeType=Capsule, length=5.75, radius=1.125, center=<2.378 0.0001 -0.02>, rotation=<rot -89.0772 179.9986 -179.9986>}]
 dyn_string_l_03 [{shapeType=Capsule, length=5.375, radius=1.125, center=<2.8808 0 -0.0276>, rotation=<rot -89.114 179.9986 -179.9986>}]
 dyn_string_l_04 [{shapeType=Capsule, length=4, radius=1.125, center=<2.7663 0 -0.0265>, rotation=<rot -89.0997 179.9986 -179.9986>}]
 dyn_string_l_05 [{shapeType=Capsule, length=0.5, radius=1.5, center=<2.7586 0.0001 -0.0262>, rotation=<rot -89.1153 179.9986 -179.9986>}]
 dyn_string_l_06 [{shapeType=Capsule, length=0.5, radius=2, center=<3.4501 0 -0.0356>, rotation=<rot -89.1675 179.9986 -179.9986>}]
 dyn_string_r_01 [{shapeType=Capsule, length=0, radius=3, center=<1 0 0>}]
 dyn_string_r_02 [{shapeType=Capsule, length=4.75, radius=1.5, center=<2.3006 0 -0.0001>, rotation=<rot 90 0 0>}]
 dyn_string_r_03 [{shapeType=Capsule, length=4.5, radius=1.75, center=<2.3006 0 -0.0001>, rotation=<rot 90 0 0>}]
 dyn_string_r_04 [{shapeType=Capsule, length=4.5, radius=1.75, center=<2.3006 0 -0.0001>, rotation=<rot 90 0 0>}]
 dyn_string_r_05 [{shapeType=Capsule, length=0.75, radius=1.25, center=<3.4009 0 0>, rotation=<rot -90 0 0>}]
 dyn_string_r_06 [{shapeType=Capsule, length=0.5, radius=2.75, center=<3.6874 0 0>}]
 head [{shapeType=Capsule, length=31, radius=2.75, center=<1.5169 -3.3885 0>}]
 neck_01 [{shapeType=Capsule, length=6.9421, radius=4.5052, rotation=<rot 90 0 0>}]
 spine_05 [
  {shapeType=Capsule, length=19.75, radius=10.25, center=<-6.7554 2.5217 -0.0001>}
  {shapeType=Box, extentX=43.5, extentY=8.25, extentZ=67, center=<7.4078 -3.6381 0>, rotation=<rot 0 -23.7498 0>}
  {shapeType=Box, extentX=12.25, extentY=8.5, extentZ=67, center=<-13.5107 4.8767 0.0003>, rotation=<rot 0 16.2516 0>}
 ]
}
constraints [
 {bone1Name=dyn_string_l_01, bone2Name=head, swing1LimitDegrees=40, swing2LimitDegrees=40, twistLimitDegrees=40}
 {bone1Name=dyn_string_l_02, bone2Name=dyn_string_l_01, swing1LimitDegrees=5, swing2LimitDegrees=5, twistLimitDegrees=5}
 {bone1Name=dyn_string_l_03, bone2Name=dyn_string_l_02, swing1LimitDegrees=5, swing2LimitDegrees=7.5, twistLimitDegrees=5}
 {bone1Name=dyn_string_l_04, bone2Name=dyn_string_l_03, swing1LimitDegrees=5, swing2LimitDegrees=15, twistLimitDegrees=5}
 {bone1Name=dyn_string_l_05, bone2Name=dyn_string_l_04, swing1LimitDegrees=5, swing2LimitDegrees=15, twistLimitDegrees=5}
 {bone1Name=dyn_string_l_06, bone2Name=dyn_string_l_05, swing1LimitDegrees=5, swing2LimitDegrees=15, twistLimitDegrees=5}
 {bone1Name=dyn_string_r_01, bone2Name=head, swing1LimitDegrees=40, swing2LimitDegrees=40, twistLimitDegrees=40}
 {bone1Name=dyn_string_r_02, bone2Name=dyn_string_r_01, swing1LimitDegrees=5, swing2LimitDegrees=5, twistLimitDegrees=5}
 {bone1Name=dyn_string_r_03, bone2Name=dyn_string_r_02, swing1LimitDegrees=5, swing2LimitDegrees=7.5, twistLimitDegrees=5}
 {bone1Name=dyn_string_r_04, bone2Name=dyn_string_r_03, swing1LimitDegrees=5, swing2LimitDegrees=15, twistLimitDegrees=5}
 {bone1Name=dyn_string_r_05, bone2Name=dyn_string_r_04, swing1LimitDegrees=5, swing2LimitDegrees=15, twistLimitDegrees=5}
 {bone1Name=dyn_string_r_06, bone2Name=dyn_string_r_05, swing1LimitDegrees=5, swing2LimitDegrees=15, twistLimitDegrees=5}
 {
  bone1Name=clavicle_r
  bone2Name=spine_05
  swing1LimitDegrees=45
  swing1Motion=Free
  swing2LimitDegrees=45
  swing2Motion=Free
  twistLimitDegrees=45
  twistMotion=Free
 }
 {
  bone1Name=clavicle_l
  bone2Name=spine_05
  swing1LimitDegrees=45
  swing1Motion=Free
  swing2LimitDegrees=45
  swing2Motion=Free
  twistLimitDegrees=45
  twistMotion=Free
 }
 {
  bone1Name=head
  bone2Name=neck_01
  swing1LimitDegrees=45
  swing1Motion=Free
  swing2LimitDegrees=45
  swing2Motion=Free
  twistLimitDegrees=45
  twistMotion=Free
 }
 {
  bone1Name=neck_01
  bone2Name=spine_05
  swing1LimitDegrees=45
  swing1Motion=Free
  swing2LimitDegrees=45
  swing2Motion=Free
  twistLimitDegrees=45
  twistMotion=Free
 }
]
