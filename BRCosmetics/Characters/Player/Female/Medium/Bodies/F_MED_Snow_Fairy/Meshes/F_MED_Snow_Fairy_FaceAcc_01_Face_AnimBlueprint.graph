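# Animation Blueprint for the Snow Fairy face accessory, adding secondary motion on top of the pose copied from its attached parent mesh.
# The pose runs through a RigidBody simulation based on spine_03, then spring-driven AnimDynamics for the right and left ear bones.
# The EventGraph maps head-versus-neck yaw to a clamped roll in backpack_rotate.
# Two dyn_string_l_01 Transform Bone nodes read that value but are disconnected from the output.
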AnimBlueprint /BRCosmetics/Characters/Player/Female/Medium/Bodies/F_MED_Snow_Fairy/Meshes/F_MED_Snow_Fairy_FaceAcc_01_Face_AnimBlueprint
target_skeleton=&/F_MED_Snow_Fairy_FaceAcc_01_Face_Skeleton
graphs [
 AnimGraph {nodes [
   "Transform (Modify) Bone - Bone: dyn_string_l_01"(Alpha=1, Translation="0,0,0", Rotation=Get backpack_rotate(), Scale="1,1,1") {rotationMode=BMM_Additive, alphaBoolBlend {blendOption=Linear}, boneToModify {boneName=dyn_string_l_01}}
   "Transform (Modify) Bone - Bone: dyn_string_l_01"(Alpha=1, Translation="0,0,0", Rotation=Get backpack_rotate(), Scale="1,1,1") {rotationMode=BMM_Additive, alphaBoolBlend {blendOption=Linear}, boneToModify {boneName=dyn_string_l_01}}
   n1 = RigidBody(ComponentPose=Local To Component(LocalPose=Copy Pose From Mesh() {bUseAttachedParent=true})) {
    bOverrideWorldGravity=true
    overridePhysicsAsset=&/F_MED_Snow_Fairy_FaceAcc_01_Physics
    simulationSpace=BaseBoneSpace
    alphaBoolBlend {blendOption=Linear}
    baseBoneRef {boneName=spine_03}
    componentLinearAccScale=<0.5 0.5 0.5>
    componentLinearVelScale=<1 1 1>
    overrideWorldGravity {z=-980}
   }
   n2 = Anim Dynamics - Bone: dyn_ear_r(ComponentPose=@n1, Alpha=0.6, ExternalForce="0,0,0") {
    alpha=0.9
    angularSpringConstant=500
    bAngularSpring=true
    bGravityOverrideInSimSpace=true
    numSolverIterationsPostUpdate=4
    numSolverIterationsPreUpdate=16
    simulationSpace=Actor
    alphaBoolBlend {blendOption=Linear}
    boundBone {boneName=dyn_ear_r}
    componentLinearAccScale=<0.15 0.15 0.15>
    componentLinearVelScale=<0.15 0.15 0.15>
    physicsBodyDefinitions [{
      sphereCollisionRadius=0
      boundBone {boneName=dyn_ear_r}
      boxExtents=<15 5 5>
      constraintSetup {angularTargetAxis=AxisZ, twistAxis=AxisY, angularLimitsMax=<15 0 15>, angularLimitsMin=<-15 0 -15>, angularTarget=<0 0 1>}
      localJointOffset=<0 0 0>
     }]
   }
   n3 = Anim Dynamics - Bone: dyn_ear_l(ComponentPose=@n2, Alpha=0.6, ExternalForce="0,0,0") {
    alpha=0.9
    angularSpringConstant=600
    bAngularSpring=true
    bGravityOverrideInSimSpace=true
    gravityScale=0.85
    numSolverIterationsPostUpdate=4
    numSolverIterationsPreUpdate=16
    simulationSpace=Actor
    alphaBoolBlend {blendOption=Linear}
    boundBone {boneName=dyn_ear_l}
    componentLinearAccScale=<0.2 0.2 0.2>
    componentLinearVelScale=<0.2 0.2 0.2>
    physicsBodyDefinitions [{
      sphereCollisionRadius=0
      boundBone {boneName=dyn_ear_l}
      boxExtents=<15 5 5>
      constraintSetup {angularTargetAxis=AxisZ, twistAxis=AxisY, angularLimitsMax=<15 15 15>, angularLimitsMin=<-15 -15 -15>, angularTarget=<0 0 1>}
      localJointOffset=<0 0 0>
     }]
   }
   Output Pose(Result=Component To Local(ComponentPose=@n3)) {layerGroup=None}
  ]}
 EventGraph {nodes [
   Event BlueprintUpdateAnimation()
   n1 = GetOwnerSkeletalMeshComponent()
   n2 = BreakRotator(InRot=BreakTransform(InTransform=GetDeltaTransformFromRefPose(self=@n1, BoneName=head, BaseName=neck_01)).Rotation)
   > Set backpack_rotate(backpack_rotate=MakeRotator(Roll=MapRangeClamped(Value=@n2.Yaw, InRangeA=-30, InRangeB=20, OutRangeA=20, OutRangeB=-40)))
  ]}
]
variables [backpack_rotate {type=Rotator}]
